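; BP_PreviewGroom is an actor Blueprint that previews a groom with its solver and an optional animated body mesh.
; The construction script assigns the DataflowAsset variable as the groom asset and registers the strands component with the GroomSolver.
; A valid binding asset supplies the body's skeletal mesh, which receives the strands as a child, plays the animation, joins the solver as collision and simulates physics.
; BeginPlay, Tick and overlap events are empty.
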(event EventBeginPlay)

(event Collision|EventActorBeginOverlap
  (params (OtherActor :type Actor%20Object%20Reference)))

(event EventTick
  (params (DeltaSeconds :type Float%20%28single-precision%29)))

(fn ConstructionScript
  (Groom|SetGroomAsset (Variables|Default|GetGroomStrands) (Variables|Default|GetDataflowAsset))
  (Groom|AddGroomComponent (Variables|Default|GetGroomSolver) (Variables|Default|GetGroomStrands))
  (if (Utilities|IsValid (Variables|Default|GetBindingAsset) :!class /Script/Engine.KismetSystemLibrary)
    (Variables|/Script/Engine.SkeletalMeshComponent|SetSkeletalMeshAsset (Variables|/Script/HairStrandsCore.GroomBindingAsset|GetTargetSkeletalMesh (Variables|Default|GetBindingAsset)) (Variables|Default|GetBodyMesh))
    (Groom|SetBindingAsset (Variables|Default|GetGroomStrands) (Variables|Default|GetBindingAsset))
    (Transformation|AttachComponentToComponent (Variables|Default|GetGroomStrands) (Variables|Default|GetBodyMesh) "None" "KeepRelative" "KeepRelative" "KeepRelative" true)
    (Components|Animation|OverrideAnimationData (Variables|Default|GetBodyMesh) (Variables|Default|GetAnimationAsset) true true 0.0 1.0)
    (Groom|AddCollisionComponent (Variables|Default|GetGroomSolver) (Variables|Default|GetBodyMesh) 0)
    (Physics|SetSimulatePhysics (Variables|Default|GetBodyMesh) true)))
