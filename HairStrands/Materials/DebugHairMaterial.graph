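# Debug hair material for two-sided, lit Niagara sprites, with geometry math that reshapes each sprite into a segment.
# Particle color RGB serves as the segment's end position, and particle position as its start.
# V runs along the segment, while U spreads width by particle size along a side vector perpendicular to the segment and world Z.
# Subtracting the unoffset world position yields a vertex offset; a constant tan color and a perpendicular cross-product vector accompany it.
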Material /HairStrands/Materials/DebugHairMaterial
graph {
 properties {bUsedWithNiagaraSprites=true, pixelDepthOffsetMode=PDOM_Legacy, twoSided=true, usedShadingModels=Default Lit}
 nodes [
  Constant3Vector() {constant=<rgba 0.75 0.6292 0.2852 0>}
  n1 = ParticlePositionWS()
  n2 = Subtract(A=ComponentMask(None=ParticleColor().RGB) {b=true, g=true, r=true}, B=@n1)
  n3 = Normalize(VectorInput=@n2)
  n4 = Normalize(VectorInput=CrossProduct(A=@n3, B=Constant3Vector() {constant=<rgba 0 0 1 0>}))
  CrossProduct(A=@n4, B=@n3)
  n5 = TextureCoordinate()
  n6 = Multiply(A=@n2, B=ComponentMask(None=@n5) {g=true})
  n7 = Multiply(A=Subtract(A=ComponentMask(None=@n5) {r=true}) {constB=0.5}, B=ComponentMask(None=ParticleSize()) {r=true}) {constB=2}
  Subtract(A=Add(A=Add(A=@n1, B=@n6), B=Multiply(A=@n4, B=@n7)), B=WorldPosition() {worldPositionShaderOffset=WPT_ExcludeAllShaderOffsets}.XYZ)
 ]
}
