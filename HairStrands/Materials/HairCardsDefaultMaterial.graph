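# Default masked, two-sided material for hair cards in the HairStrands plugin, flagged for hair strands usage.
# The shading model comes from the material expression: a feature level switch selects Hair by default and falls back to Default Lit on ES3.1.
# The graph holds Color and Roughness parameters for tuning the look, plus a Hair Attributes node for per-hair data.
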Material /HairStrands/Materials/HairCardsDefaultMaterial
graph {
 properties {
  bUsedWithHairStrands=true
  blendMode=BLEND_Masked
  opacityMaskClipValue=0.01
  pixelDepthOffsetMode=PDOM_Legacy
  shadingModel=MSM_FromMaterialExpression
  twoSided=true
  usedShadingModels=Default Lit | Hair
 }
 nodes [
  FeatureLevelSwitch(Default=ShadingModel() {shadingModel=MSM_Hair}, ES3_1=ShadingModel())
  HairAttributes()
  ScalarParameter() {defaultValue=0.333, parameterName=Roughness}
  VectorParameter() {parameterName=Color, defaultValue=<rgba 0.2395 0.1862 0.1072 0>}
 ]
}
