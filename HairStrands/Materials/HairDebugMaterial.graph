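# Debug material for hair strands that visualizes per-strand attributes such as seed, UVs, radius, base color, roughness, group, clump ID and AO.
# FloatProperty selects the mode through parallel If nodes merged by a Max chain, with generically named parameters supplying mode arguments.
# Several modes use color lookup textures; the custom clump hash uses logical && where bit masks appear intended.
# A feature-level switch picks Hair shading, or Default Lit on ES3_1.
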Material /HairStrands/Materials/HairDebugMaterial
graph {
 expressions {
  MaterialExpressionAdd_0 {
   class=MaterialExpressionAdd
   inputs [
    A {from=MaterialExpressionHairAttributes_0, output=Group Index}
    B {from=""}
   ]
   properties {constB=0.5}
  }
  MaterialExpressionComponentMask_0 {class=MaterialExpressionComponentMask, inputs [None {from=MaterialExpressionHairAttributes_0, output=Root UV}], properties {r=true}}
  MaterialExpressionComponentMask_1 {class=MaterialExpressionComponentMask, inputs [None {from=MaterialExpressionHairAttributes_0, output=Root UV}], properties {g=true}}
  MaterialExpressionComponentMask_2 {class=MaterialExpressionComponentMask, inputs [None {from=MaterialExpressionReroute_1}], properties {r=true}}
  MaterialExpressionComponentMask_3 {class=MaterialExpressionComponentMask, inputs [None {from=MaterialExpressionReroute_1}], properties {g=true}}
  MaterialExpressionComponentMask_4 {class=MaterialExpressionComponentMask, inputs [None {from=MaterialExpressionReroute_1}], properties {b=true}}
  MaterialExpressionConstant3Vector_0 {class=MaterialExpressionConstant3Vector}
  MaterialExpressionConstant3Vector_1 {class=MaterialExpressionConstant3Vector}
  MaterialExpressionConstant3Vector_10 {class=MaterialExpressionConstant3Vector}
  MaterialExpressionConstant3Vector_11 {class=MaterialExpressionConstant3Vector}
  MaterialExpressionConstant3Vector_12 {class=MaterialExpressionConstant3Vector}
  MaterialExpressionConstant3Vector_2 {class=MaterialExpressionConstant3Vector}
  MaterialExpressionConstant3Vector_3 {class=MaterialExpressionConstant3Vector}
  MaterialExpressionConstant3Vector_4 {class=MaterialExpressionConstant3Vector}
  MaterialExpressionConstant3Vector_5 {class=MaterialExpressionConstant3Vector}
  MaterialExpressionConstant3Vector_6 {class=MaterialExpressionConstant3Vector}
  MaterialExpressionConstant3Vector_7 {class=MaterialExpressionConstant3Vector}
  MaterialExpressionConstant3Vector_8 {class=MaterialExpressionConstant3Vector}
  MaterialExpressionConstant3Vector_9 {class=MaterialExpressionConstant3Vector}
  MaterialExpressionConstant_0 {class=MaterialExpressionConstant, properties {r=0.5}}
  MaterialExpressionConstant_1 {class=MaterialExpressionConstant, properties {r=0.5}}
  MaterialExpressionConstant_10 {class=MaterialExpressionConstant, properties {r=1}}
  MaterialExpressionConstant_11 {class=MaterialExpressionConstant, properties {r=0.5}}
  MaterialExpressionConstant_2 {class=MaterialExpressionConstant}
  MaterialExpressionConstant_3 {class=MaterialExpressionConstant, properties {r=0.5}}
  MaterialExpressionConstant_4 {class=MaterialExpressionConstant}
  MaterialExpressionConstant_5 {class=MaterialExpressionConstant, properties {r=1}}
  MaterialExpressionConstant_6 {class=MaterialExpressionConstant}
  MaterialExpressionConstant_7 {class=MaterialExpressionConstant, properties {r=0.5}}
  MaterialExpressionConstant_8 {class=MaterialExpressionConstant, properties {r=255}}
  MaterialExpressionConstant_9 {class=MaterialExpressionConstant, properties {r=0.5}}
  MaterialExpressionCustom_0 {
   class=MaterialExpressionCustom
   inputs [InClumpIndex {from=MaterialExpressionSwitch_0, output=""}]
   properties {
    code="{\r\n\tuint In2 = InClumpIndex;\r\n\r\n\tconst uint InLo8bits =  In2 && 0xFF;\r\n\tconst uint InHi8btis = (In2 >> 8) && 0xFF;\r\n\r\n\t// Using FNV1 hash to break linearity of ClusterId (generated by linearly parsing cluster grid)\r\n\tuint Hash = 0;\r\n\tHash = Hash * 17;\r\n\tHash = Hash ^ InLo8bits;\r\n\tHash = Hash * 17;\r\n\tHash = Hash ^ InHi8btis;\r\n\tHash = (Hash && 0xFF);\r\n\r\n\treturn float(Hash) / 255.f;\r\n}"
    outputType=CMOT_Float1
    inputs [{
      inputName=InClumpIndex
      input {
       expression=&self:MaterialExpressionSwitch_0
       inputName=None
       mask=0
       maskA=0
       maskB=0
       maskG=0
       maskR=0
       outputIndex=0
      }
     }]
   }
  }
  MaterialExpressionDivide_0 {
   class=MaterialExpressionDivide
   inputs [
    A {from=MaterialExpressionFmod_0}
    B {from=""}
   ]
   properties {constB=255}
  }
  MaterialExpressionDivide_1 {
   class=MaterialExpressionDivide
   inputs [
    A {from=MaterialExpressionSubtract_1}
    B {from=MaterialExpressionSubtract_0}
   ]
  }
  MaterialExpressionDivide_2 {
   class=MaterialExpressionDivide
   inputs [
    A {from=MaterialExpressionAdd_0}
    B {from=""}
   ]
   properties {constB=12}
  }
  MaterialExpressionFeatureLevelSwitch_0 {
   class=MaterialExpressionFeatureLevelSwitch
   inputs [
    Default {from=MaterialExpressionShadingModel_0, output=""}
    ES2 {from="", output=""}
    ES3_1 {from=MaterialExpressionShadingModel_1, output=""}
    SM4_REMOVED {from="", output=""}
    SM5 {from="", output=""}
    SM6 {from="", output=""}
   ]
   properties {}
  }
  MaterialExpressionFmod_0 {
   class=MaterialExpressionFmod
   inputs [
    A {from=MaterialExpressionMultiply_2, output=""}
    B {from=MaterialExpressionConstant_8, output=""}
   ]
   properties {}
  }
  MaterialExpressionFrac_0 {class=MaterialExpressionFrac, inputs [None {from=MaterialExpressionMultiply_0, output=""}], properties {}}
  MaterialExpressionFrac_1 {class=MaterialExpressionFrac, inputs [None {from=MaterialExpressionMultiply_1, output=""}], properties {}}
  MaterialExpressionFrac_2 {class=MaterialExpressionFrac, inputs [None {from=MaterialExpressionHairAttributes_0, output=Root UV}], properties {}}
  MaterialExpressionHairAttributes_0 {class=MaterialExpressionHairAttributes, inputs [], properties {}}
  MaterialExpressionIf_0 {
   class=MaterialExpressionIf
   inputs [
    A {from=MaterialExpressionScalarParameter_0, output=""}
    B {from="", output=""}
    "A > B" {from=MaterialExpressionConstant3Vector_0, output=""}
    "A == B" {from=MaterialExpressionTextureSample_1, output=RGB}
    "A < B" {from=MaterialExpressionConstant3Vector_0, output=""}
   ]
   properties {}
  }
  MaterialExpressionIf_1 {
   class=MaterialExpressionIf
   inputs [
    A {from=MaterialExpressionScalarParameter_0, output=""}
    B {from="", output=""}
    "A > B" {from=MaterialExpressionConstant3Vector_1, output=""}
    "A == B" {from=MaterialExpressionMaterialFunctionCall_3, output=Result}
    "A < B" {from=MaterialExpressionConstant3Vector_1, output=""}
   ]
   properties {constB=2}
  }
  MaterialExpressionIf_10 {
   class=MaterialExpressionIf
   inputs [
    A {from=MaterialExpressionScalarParameter_0, output=""}
    B {from="", output=""}
    "A > B" {from=MaterialExpressionConstant3Vector_9, output=""}
    "A == B" {from=MaterialExpressionTextureSample_5, output=RGB}
    "A < B" {from=MaterialExpressionConstant3Vector_9, output=""}
   ]
   properties {constB=9}
  }
  MaterialExpressionIf_11 {
   class=MaterialExpressionIf
   inputs [
    A {from=MaterialExpressionScalarParameter_0, output=""}
    B {from="", output=""}
    "A > B" {from=MaterialExpressionReroute_0, output=""}
    "A == B" {from=MaterialExpressionConstant_10, output=""}
    "A < B" {from=MaterialExpressionReroute_0, output=""}
   ]
   properties {constB=10}
  }
  MaterialExpressionIf_12 {
   class=MaterialExpressionIf
   inputs [
    A {from=MaterialExpressionScalarParameter_0, output=""}
    B {from="", output=""}
    "A > B" {from=MaterialExpressionConstant3Vector_10, output=""}
    "A == B" {from=MaterialExpressionMultiply_3, output=""}
    "A < B" {from=MaterialExpressionConstant3Vector_10, output=""}
   ]
   properties {constB=10}
  }
  MaterialExpressionIf_13 {
   class=MaterialExpressionIf
   inputs [
    A {from=MaterialExpressionScalarParameter_0, output=""}
    B {from="", output=""}
    "A > B" {from=MaterialExpressionConstant3Vector_11, output=""}
    "A == B" {from=MaterialExpressionTextureSample_4, output=RGB}
    "A < B" {from=MaterialExpressionConstant3Vector_11, output=""}
   ]
   properties {constB=11}
  }
  MaterialExpressionIf_14 {
   class=MaterialExpressionIf
   inputs [
    A {from=MaterialExpressionScalarParameter_0, output=""}
    B {from="", output=""}
    "A > B" {from=MaterialExpressionConstant3Vector_12, output=""}
    "A == B" {from=MaterialExpressionMaterialFunctionCall_9, output=Result}
    "A < B" {from=MaterialExpressionConstant3Vector_12, output=""}
   ]
   properties {constB=12}
  }
  MaterialExpressionIf_2 {
   class=MaterialExpressionIf
   inputs [
    A {from=MaterialExpressionScalarParameter_0, output=""}
    B {from="", output=""}
    "A > B" {from=MaterialExpressionConstant3Vector_2, output=""}
    "A == B" {from=MaterialExpressionMaterialFunctionCall_4, output=Result}
    "A < B" {from=MaterialExpressionConstant3Vector_2, output=""}
   ]
   properties {constB=1}
  }
  MaterialExpressionIf_3 {
   class=MaterialExpressionIf
   inputs [
    A {from=MaterialExpressionScalarParameter_0, output=""}
    B {from="", output=""}
    "A > B" {from=MaterialExpressionConstant3Vector_3, output=""}
    "A == B" {from=MaterialExpressionTextureSample_0, output=RGB}
    "A < B" {from=MaterialExpressionConstant3Vector_3, output=""}
   ]
   properties {constB=3}
  }
  MaterialExpressionIf_4 {
   class=MaterialExpressionIf
   inputs [
    A {from=MaterialExpressionScalarParameter_0, output=""}
    B {from="", output=""}
    "A > B" {from=MaterialExpressionConstant3Vector_4, output=""}
    "A == B" {from=MaterialExpressionMaterialFunctionCall_1, output=Result}
    "A < B" {from=MaterialExpressionConstant3Vector_4, output=""}
   ]
   properties {constB=4}
  }
  MaterialExpressionIf_5 {
   class=MaterialExpressionIf
   inputs [
    A {from=MaterialExpressionScalarParameter_0, output=""}
    B {from="", output=""}
    "A > B" {from=MaterialExpressionConstant3Vector_5, output=""}
    "A == B" {from=MaterialExpressionTextureSample_2, output=RGB}
    "A < B" {from=MaterialExpressionConstant3Vector_5, output=""}
   ]
   properties {constB=5}
  }
  MaterialExpressionIf_6 {
   class=MaterialExpressionIf
   inputs [
    A {from=MaterialExpressionScalarParameter_2, output=""}
    B {from=MaterialExpressionHairAttributes_0, output=Length}
    "A > B" {from=MaterialExpressionConstant_5, output=""}
    "A == B" {from=MaterialExpressionConstant_5, output=""}
    "A < B" {from=MaterialExpressionConstant_4, output=""}
   ]
   properties {}
  }
  MaterialExpressionIf_7 {
   class=MaterialExpressionIf
   inputs [
    A {from=MaterialExpressionScalarParameter_0, output=""}
    B {from="", output=""}
    "A > B" {from=MaterialExpressionConstant3Vector_6, output=""}
    "A == B" {from=MaterialExpressionTextureSample_3, output=RGB}
    "A < B" {from=MaterialExpressionConstant3Vector_6, output=""}
   ]
   properties {constB=6}
  }
  MaterialExpressionIf_8 {
   class=MaterialExpressionIf
   inputs [
    A {from=MaterialExpressionScalarParameter_0, output=""}
    B {from="", output=""}
    "A > B" {from=MaterialExpressionConstant3Vector_7, output=""}
    "A == B" {from=MaterialExpressionHairAttributes_0, output=BaseColor}
    "A < B" {from=MaterialExpressionConstant3Vector_7, output=""}
   ]
   properties {constB=7}
  }
  MaterialExpressionIf_9 {
   class=MaterialExpressionIf
   inputs [
    A {from=MaterialExpressionScalarParameter_0, output=""}
    B {from="", output=""}
    "A > B" {from=MaterialExpressionConstant3Vector_8, output=""}
    "A == B" {from=MaterialExpressionMaterialFunctionCall_6, output=Result}
    "A < B" {from=MaterialExpressionConstant3Vector_8, output=""}
   ]
   properties {constB=8}
  }
  MaterialExpressionMaterialFunctionCall_0 {
   class=MaterialExpressionMaterialFunctionCall
   inputs [
    X {from=MaterialExpressionHairAttributes_0, output=Seed}
    Y {from=MaterialExpressionConstant_1}
   ]
   properties {materialFunction=&MakeFloat2}
  }
  MaterialExpressionMaterialFunctionCall_1 {
   class=MaterialExpressionMaterialFunctionCall
   inputs [
    X {from=MaterialExpressionFrac_0}
    Y {from=MaterialExpressionFrac_1}
    Z {from=MaterialExpressionConstant_6}
   ]
   properties {materialFunction=&MakeFloat3}
  }
  MaterialExpressionMaterialFunctionCall_2 {
   class=MaterialExpressionMaterialFunctionCall
   inputs [
    X {from=MaterialExpressionHairAttributes_0, output=Seed}
    Y {from=MaterialExpressionConstant_0}
   ]
   properties {materialFunction=&MakeFloat2}
  }
  MaterialExpressionMaterialFunctionCall_3 {
   class=MaterialExpressionMaterialFunctionCall
   inputs [
    X {from=MaterialExpressionHairAttributes_0, output=U}
    Y {from=MaterialExpressionHairAttributes_0, output=U}
    Z {from=""}
   ]
   properties {materialFunction=&MakeFloat3}
  }
  MaterialExpressionMaterialFunctionCall_4 {
   class=MaterialExpressionMaterialFunctionCall
   inputs [
    X {from=MaterialExpressionComponentMask_0}
    Y {from=MaterialExpressionComponentMask_1}
    Z {from=MaterialExpressionConstant_2}
   ]
   properties {materialFunction=&MakeFloat3}
  }
  MaterialExpressionMaterialFunctionCall_5 {
   class=MaterialExpressionMaterialFunctionCall
   inputs [
    X {from=MaterialExpressionSaturate_1}
    Y {from=MaterialExpressionConstant_3}
   ]
   properties {materialFunction=&MakeFloat2}
  }
  MaterialExpressionMaterialFunctionCall_6 {
   class=MaterialExpressionMaterialFunctionCall
   inputs [
    X {from=MaterialExpressionHairAttributes_0, output=Roughness}
    Y {from=MaterialExpressionHairAttributes_0, output=Roughness}
    Z {from=MaterialExpressionHairAttributes_0, output=Roughness}
   ]
   properties {materialFunction=&MakeFloat3}
  }
  MaterialExpressionMaterialFunctionCall_7 {
   class=MaterialExpressionMaterialFunctionCall
   inputs [
    X {from=MaterialExpressionDivide_2}
    Y {from=MaterialExpressionConstant_9}
   ]
   properties {materialFunction=&MakeFloat2}
  }
  MaterialExpressionMaterialFunctionCall_8 {
   class=MaterialExpressionMaterialFunctionCall
   inputs [
    X {from=MaterialExpressionCustom_0}
    Y {from=MaterialExpressionConstant_11}
   ]
   properties {materialFunction=&MakeFloat2}
  }
  MaterialExpressionMaterialFunctionCall_9 {
   class=MaterialExpressionMaterialFunctionCall
   inputs [
    X {from=MaterialExpressionHairAttributes_0, output=AO}
    Y {from=MaterialExpressionHairAttributes_0, output=AO}
    Z {from=MaterialExpressionHairAttributes_0, output=AO}
   ]
   properties {materialFunction=&MakeFloat3}
  }
  MaterialExpressionMax_0 {
   class=MaterialExpressionMax
   inputs [
    A {from=MaterialExpressionIf_0}
    B {from=MaterialExpressionMax_1}
   ]
  }
  MaterialExpressionMax_1 {
   class=MaterialExpressionMax
   inputs [
    A {from=MaterialExpressionIf_2}
    B {from=MaterialExpressionMax_2}
   ]
  }
  MaterialExpressionMax_10 {
   class=MaterialExpressionMax
   inputs [
    A {from=MaterialExpressionIf_12}
    B {from=MaterialExpressionMax_11}
   ]
  }
  MaterialExpressionMax_11 {
   class=MaterialExpressionMax
   inputs [
    A {from=MaterialExpressionIf_13}
    B {from=MaterialExpressionIf_14}
   ]
  }
  MaterialExpressionMax_2 {
   class=MaterialExpressionMax
   inputs [
    A {from=MaterialExpressionIf_1}
    B {from=MaterialExpressionMax_3}
   ]
  }
  MaterialExpressionMax_3 {
   class=MaterialExpressionMax
   inputs [
    A {from=MaterialExpressionIf_3}
    B {from=MaterialExpressionMax_4}
   ]
  }
  MaterialExpressionMax_4 {
   class=MaterialExpressionMax
   inputs [
    A {from=MaterialExpressionIf_4}
    B {from=MaterialExpressionMax_5}
   ]
  }
  MaterialExpressionMax_5 {
   class=MaterialExpressionMax
   inputs [
    A {from=MaterialExpressionIf_5}
    B {from=MaterialExpressionMax_6}
   ]
  }
  MaterialExpressionMax_6 {
   class=MaterialExpressionMax
   inputs [
    A {from=MaterialExpressionIf_7}
    B {from=MaterialExpressionMax_7}
   ]
  }
  MaterialExpressionMax_7 {
   class=MaterialExpressionMax
   inputs [
    A {from=MaterialExpressionIf_8}
    B {from=MaterialExpressionMax_8}
   ]
  }
  MaterialExpressionMax_8 {
   class=MaterialExpressionMax
   inputs [
    A {from=MaterialExpressionIf_9}
    B {from=MaterialExpressionMax_10}
   ]
  }
  MaterialExpressionMax_9 {
   class=MaterialExpressionMax
   inputs [
    A {from=MaterialExpressionMax_0}
    B {from=MaterialExpressionIf_10}
   ]
  }
  MaterialExpressionMultiply_0 {
   class=MaterialExpressionMultiply
   inputs [
    A {from=""}
    B {from=MaterialExpressionHairAttributes_0, output=Length}
   ]
   properties {constA=1}
  }
  MaterialExpressionMultiply_1 {
   class=MaterialExpressionMultiply
   inputs [
    A {from=""}
    B {from=MaterialExpressionHairAttributes_0, output=Radius}
   ]
   properties {constA=1}
  }
  MaterialExpressionMultiply_2 {
   class=MaterialExpressionMultiply
   inputs [
    A {from=MaterialExpressionSubtract_2}
    B {from=""}
   ]
   properties {constA=1, constB=23}
  }
  MaterialExpressionMultiply_3 {
   class=MaterialExpressionMultiply
   inputs [
    A {from=MaterialExpressionVectorParameter_0, output=RGB}
    B {from=MaterialExpressionConstant_7}
   ]
  }
  MaterialExpressionReroute_0 {class=MaterialExpressionReroute, inputs [None {from=MaterialExpressionHairAttributes_0, output=Coverage}]}
  MaterialExpressionReroute_1 {class=MaterialExpressionReroute, inputs [None {from=MaterialExpressionHairAttributes_0, output=Clump ID}]}
  MaterialExpressionReroute_2 {class=MaterialExpressionReroute, inputs [None {from=MaterialExpressionScalarParameter_3}]}
  MaterialExpressionSaturate_1 {class=MaterialExpressionSaturate, inputs [None {from=MaterialExpressionDivide_1}]}
  MaterialExpressionScalarParameter_0 {
   class=MaterialExpressionScalarParameter
   properties {desc="Define which debug mode is run: 0:ClusterColor, 1:HairUV, 2:RootUV, 3:Seed, 4:Tangent", parameterName=FloatProperty, sliderMax=4}
  }
  MaterialExpressionScalarParameter_1 {
   class=MaterialExpressionScalarParameter
   properties {desc=Called ByteProperty as a workaround for passing named parameters, parameterName=ByteProperty}
  }
  MaterialExpressionScalarParameter_2 {class=MaterialExpressionScalarParameter, properties {parameterName=BoolProperty}}
  MaterialExpressionScalarParameter_3 {
   class=MaterialExpressionScalarParameter
   properties {defaultValue=0.01, desc=Called IntProperty as a workaround for passing named parameters, parameterName=IntProperty}
  }
  MaterialExpressionScalarParameter_4 {class=MaterialExpressionScalarParameter, properties {defaultValue=1, parameterName=Roughness}}
  MaterialExpressionShadingModel_0 {class=MaterialExpressionShadingModel, inputs [], properties {shadingModel=MSM_Hair}}
  MaterialExpressionShadingModel_1 {class=MaterialExpressionShadingModel, inputs [], properties {}}
  MaterialExpressionSubtract_0 {
   class=MaterialExpressionSubtract
   inputs [
    A {from=MaterialExpressionReroute_2}
    B {from=MaterialExpressionScalarParameter_1}
   ]
  }
  MaterialExpressionSubtract_1 {
   class=MaterialExpressionSubtract
   inputs [
    A {from=MaterialExpressionHairAttributes_0, output=Radius}
    B {from=MaterialExpressionScalarParameter_1}
   ]
  }
  MaterialExpressionSubtract_2 {
   class=MaterialExpressionSubtract
   inputs [
    A {from=MaterialExpressionHairAttributes_0, output=Root UV}
    B {from=MaterialExpressionFrac_2}
   ]
  }
  MaterialExpressionSwitch_0 {
   class=MaterialExpressionSwitch
   inputs [
    SwitchValue {from=MaterialExpressionReroute_2, output=""}
    Default {from="", output=""}
    None {from=MaterialExpressionComponentMask_2, output=""}
    None {from=MaterialExpressionComponentMask_3, output=""}
    None {from=MaterialExpressionComponentMask_4, output=""}
   ]
   properties {inputs [
     {
      inputName=None
      input {
       expression=&self:MaterialExpressionComponentMask_2
       inputName=None
       mask=0
       maskA=0
       maskB=0
       maskG=0
       maskR=0
       outputIndex=0
      }
     }
     {
      inputName=None
      input {
       expression=&self:MaterialExpressionComponentMask_3
       inputName=None
       mask=0
       maskA=0
       maskB=0
       maskG=0
       maskR=0
       outputIndex=0
      }
     }
     {
      inputName=None
      input {
       expression=&self:MaterialExpressionComponentMask_4
       inputName=None
       mask=0
       maskA=0
       maskB=0
       maskG=0
       maskR=0
       outputIndex=0
      }
     }
    ]}
  }
  MaterialExpressionTextureSample_0 {
   class=MaterialExpressionTextureSample
   inputs [
    UVs {from=MaterialExpressionMaterialFunctionCall_0, output=Result}
    Tex {from=""}
    Apply View MipBias {from=""}
   ]
   properties {samplerSource=SSM_Clamp_WorldGroupSettings, texture=/Engine/EngineMaterials/HairDebugColor}
  }
  MaterialExpressionTextureSample_1 {
   class=MaterialExpressionTextureSample
   inputs [
    UVs {from=MaterialExpressionMaterialFunctionCall_2, output=Result}
    Tex {from=""}
    Apply View MipBias {from=""}
   ]
   properties {samplerSource=SSM_Clamp_WorldGroupSettings, texture=&HairDebugColor}
  }
  MaterialExpressionTextureSample_2 {
   class=MaterialExpressionTextureSample
   inputs [
    UVs {from=MaterialExpressionMaterialFunctionCall_5, output=Result}
    Tex {from=""}
    Apply View MipBias {from=""}
   ]
   properties {texture=/Engine/EngineMaterials/HairColorMap}
  }
  MaterialExpressionTextureSample_3 {
   class=MaterialExpressionTextureSample
   inputs [
    UVs {from=MaterialExpressionDivide_0}
    Tex {from=""}
    Apply View MipBias {from=""}
   ]
   properties {samplerSource=SSM_Clamp_WorldGroupSettings, texture=&HairDebugColor}
  }
  MaterialExpressionTextureSample_4 {
   class=MaterialExpressionTextureSample
   inputs [
    UVs {from=MaterialExpressionMaterialFunctionCall_8, output=Result}
    Tex {from=""}
    Apply View MipBias {from=""}
   ]
   properties {samplerSource=SSM_Clamp_WorldGroupSettings, texture=&HairDebugColor}
  }
  MaterialExpressionTextureSample_5 {
   class=MaterialExpressionTextureSample
   inputs [
    UVs {from=MaterialExpressionMaterialFunctionCall_7, output=Result}
    Tex {from=""}
    Apply View MipBias {from=""}
   ]
   properties {texture=&/HairGroupColors}
  }
  MaterialExpressionVectorParameter_0 {class=MaterialExpressionVectorParameter, properties {parameterName=VectorProperty}}
 }
 properties {
  bUsedWithHairStrands=true
  blendMode=BLEND_Masked
  pixelDepthOffsetMode=PDOM_Legacy
  shadingModel=MSM_FromMaterialExpression
  twoSided=true
  usedShadingModels=Default Lit | Hair
 }
}
