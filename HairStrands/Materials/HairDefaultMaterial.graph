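# Default hair strands material for the HairStrands content, a two-sided surface using the Hair shading model and flagged for use with hair strands.
# It declares two parameters for instances to override: a brownish Color vector and a scalar Roughness.
# The graph has no texture or math nodes, and shows no connections from either parameter to the material outputs.
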Material /HairStrands/Materials/HairDefaultMaterial
graph {
 properties {bUsedWithHairStrands=true, pixelDepthOffsetMode=PDOM_Legacy, shadingModel=MSM_Hair, twoSided=true, usedShadingModels=Hair}
 nodes [
  ScalarParameter() {defaultValue=0.333, parameterName=Roughness}
  VectorParameter() {parameterName=Color, defaultValue=<rgba 0.2395 0.1862 0.1072 0>}
 ]
}
