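# Minimal Niagara hair material in the HairStrands content, flagged for use on Niagara mesh particles.
# It uses the Default Lit shading model with legacy pixel depth offset mode.
# Its graph holds a single Particle Color node, with no connection to any material output shown.
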Material /HairStrands/Materials/NiagaraHairMaterial
graph {properties {bUsedWithNiagaraMeshParticles=true, pixelDepthOffsetMode=PDOM_Legacy, usedShadingModels=Default Lit}, nodes [ParticleColor()]}
